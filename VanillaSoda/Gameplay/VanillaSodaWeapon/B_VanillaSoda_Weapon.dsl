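; Melee weapon Blueprint for the VanillaSoda blade, handling equip effects, sprint modifiers and melee cue feedback.
; Equipping attaches body aura, adds a custom tactical-sprint class, grants fall-damage immunity on sprint-jumps and starts cosmetic cue and ambient effects.
; Unequip and EndPlay undo most of this, with a scalable-float fixup guarding duplicate sprint removal.
; Appear, disappear and heavy-attack visuals drive dynamic-material parameters stamped with game time.
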(event FortWeapon|EventOnEquippedAndAttached
  (bind _returnvalue (Game|GetInstigator))
  (bind _asfort_player_pawn_athena (Utilities|Casting|CastToFortPlayerPawnAthena _returnvalue))
  (bind _instigator (Variables|Actor|GetInstigator))
  (bind _mesh (Variables|/Script/Engine.Character|GetMesh _asfort_player_pawn_athena))
  ;[0.05381 0.15 0.1098 0.5] On Equip
  (|CallParentFunction)
  (Utilities|FlowControl|Sequence
    (:then_0
      (Variables|Default|SetPlayerPawn _asfort_player_pawn_athena)
      (Transformation|AttachComponentToComponent (Variables|Default|GetNS_IBlade_AuraBody) _mesh "None" "KeepRelative" "KeepRelative" "KeepWorld" true)
      (Variables|Default|SetCharacterMesh _mesh)
      (CallFunction|PlayAppearTA 0.0) ; Reset to hide
      (CallFunction|PlayAppearTA 1.0 0.5) ; Play appear
      (bind _output_get (Variables|Default|SetTactical%20Sprint%20Component (Actor|GetComponentbyClass (Game|GetInstigatorController) "/Script/TacticalSprintRuntime.FortPlayerControllerComponent_TacticalSprint")))
      (Sprint|AddCustomSprintMMEClass _output_get (Variables|Default|GetSprint%20MMEClass%20to%20Add))
      (bind _asynctaskproxy (Ability|Tasks|WaitGameplayTagAddToActor _instigator "(TagName=\"Gameplay.Action.Player.SprintJumping\")")
        (:then
          (Variables|Transient|SetAsyncAction_WaitSprintJump _asynctaskproxy))
        (:Added
          (Fortnite|ApplyGameplayEffectsToActor _instigator (Utilities|Array|MakeArray (Variables|Settings|GetFallDamageImmunityGameplayEffectClass)) (0.0 0.0 0.0) false 0 (Ability|GetFortAbilitySystemComponent _instigator) (GameplayEffects|MakeEffectContext (Ability|GetFortAbilitySystemComponent _instigator))))))
    (:then_1
      (Networking|CanExecuteCosmeticEvents
        (:True
          (if (not (Vehicle|IsPassengerinVehicle (Variables|Default|GetPlayerPawn)))
            (GameplayCue|AddGameplayCueOnActor(Looping) self (Variables|Default|GetGameplay%20Cue%20Tag) (Ability|GameplayCue|MakeGameplayCueParameters 0.0))
            (Components|Activation|Activate (Variables|Default|GetNS_LemonExam_Ambient) true)))))))

(event FortWeapon|EventOnUnEquippedAndDetached
  ;[0.15 0.07841 0.01641 0.5] Cleanup: On Un-Equip + End Play
  (|CallParentFunction)
  (Networking|CanExecuteCosmeticEvents
    (:True
      (GameplayCue|RemoveGameplayCueOnActor(Looping) self (Variables|Default|GetGameplay%20Cue%20Tag) (Ability|GameplayCue|MakeGameplayCueParameters 0.0))
      (Components|Activation|Deactivate (Variables|Default|GetNS_LemonExam_Ambient)))))

(event FortWeapon|EventOnLocalUnEquip
  ;[0.15 0.07841 0.01641 0.5] Cleanup: On Un-Equip + End Play
  (|CallParentFunction)
  (CallFunction|PlayAppearTA 2.0 0.5)
  (Sprint|RemoveCustomSprintMMEClass (Variables|Default|GetTactical%20Sprint%20Component) (Variables|Default|GetSprint%20MMEClass%20to%20Add))
  (bind _asyncaction_waitsprintjump (Variables|Transient|GetAsyncAction_WaitSprintJump)
    (:then
      (Ability|Async|EndAction _asyncaction_waitsprintjump)
      (Components|Activation|Deactivate (Variables|Default|GetNS_LemonExam_Ambient)))
    (:else)))

(event EventEndPlay
  (params (EndPlayReason :type EEndPlayReason%20Enum :default "Destroyed"))
  ;[0.15 0.07841 0.01641 0.5] Cleanup: On Un-Equip + End Play
  (|CallParentFunction EndPlayReason)
  (if (not (and (> (ScalableFloat|GetCachedScalableFloatValue (Variables|Default|GetFixup%20Sprint%20MME%20Removal)) 0) (not (FortWeapon|IsEquipped))))
    (CallFunction|PlayAppearTA 2.0 0.5)
    (Sprint|RemoveCustomSprintMMEClass (Variables|Default|GetTactical%20Sprint%20Component) (Variables|Default|GetSprint%20MMEClass%20to%20Add)))
  (bind _asyncaction_waitsprintjump (Variables|Transient|GetAsyncAction_WaitSprintJump)
    (:then
      (Ability|Async|EndAction _asyncaction_waitsprintjump)
      (Components|Activation|Deactivate (Variables|Default|GetNS_LemonExam_Ambient)))
    (:else)))

(event Custom|PlayHeavyAttack_TA
  (bind _weaponmesh (Variables|Visual|GetWeaponMesh))
  ; PlayHeavyAttack_TA
  (Rendering|Material|SetScalarParameterValueonMaterials _weaponmesh "ATK_AnimationDuration" 2.0) ; Set Duration. 2s
  (Rendering|Material|SetScalarParameterValueonMaterials _weaponmesh "ATK_Stage" 1.0) ; Stage1
  (Rendering|Material|SetScalarParameterValueonMaterials _weaponmesh "ATK_GameTimeFromBP" (Utilities|Time|GetGameTimeinSeconds)) ; Passthough Time
)

(event Custom|PlayAppear_TA
  (params (Stage :type Float%20%28double-precision%29) (Duration :type Float%20%28single-precision%29))
  (bind _weaponmesh (Variables|Visual|GetWeaponMesh))
  ; PlayEquip_TA
  (Rendering|Material|SetScalarParameterValueonMaterials _weaponmesh "APR_AnimationDuration" Duration) ; Set Duration. 1s
  (Rendering|Material|SetScalarParameterValueonMaterials _weaponmesh "APR_Stage" Stage) ; Stage\n0:Inital (Hide).\n1:Appear.\n2:Disappear
  (Rendering|Material|SetScalarParameterValueonMaterials _weaponmesh "APR_GameTimeFromBP" (Utilities|Time|GetGameTimeinSeconds)) ; Passthough Time
)

(event FortWeapon|EventOnPlayImpactFX
  (params (HitResult :type Hit%20Result%20Structure%20%28by%20ref%29) (ImpactPhysicalSurface :type EPhysicalSurface%20Enum :default "SurfaceType_Default") (SpawnedPSC :type FXSystem%20Component%20Object%20Reference))
  ;[0.0745 0.01908 0.15 0.5] Directional Slash Decal set
  (|CallParentFunction HitResult ImpactPhysicalSurface SpawnedPSC))

(event GameplayCue|GameplayCue.VanillaSoda.Melee
  (params (EventType :type EGameplayCueEvent%20Enum :default "OnActive") (Parameters :type Gameplay%20Cue%20Parameters%20Structure%20%28by%20ref%29))
  ;[0.0745 0.01908 0.15 0.5] Directional Slash Decal set
  (Default|CallOnMelee)
  (SoundIndicator|ShowSoundIndicator (Variables|SoundIndicator|GetSoundIndicatorComponent) (Transformation|GetActorLocation (Game|GetInstigator)) "/Game/Sounds/Fort_GamePlay_Sounds/BRSword/AB_Swing_First_Cue.AB_Swing_First_Cue" 0.3 (Variables|/Script/Engine.Controller|GetPlayerState (Game|GetInstigatorController))))

(fn ConstructionScript
  (|CallParentFunction)
  (Rendering|Material|SetMaterial (Variables|Visual|GetWeaponMesh) 0 (Class|BTechArtGlobalLibrary|GetDynamicMaterialInstance (Variables|Visual|GetWeaponMesh))))
